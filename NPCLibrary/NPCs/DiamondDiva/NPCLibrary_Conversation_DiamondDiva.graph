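# Conversation database for the Diamond Diva NPC, defining the greeting and dialogue options players get when talking to this character.
# The entry point leads to one speech node with a default line, plus prioritized contextual lines for a first-ever meeting or specific player outfits.
# The speech node then branches through link nodes to shared NPC quests, basic services and a Season 26 Snapshot quest conversation, each referenced by entry tag.
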ConversationDatabase /NPCLibrary/NPCs/DiamondDiva/NPCLibrary_Conversation_DiamondDiva
entry_points [{tag=Conversation.Entry.DiamondDiva, destinations [#1]}]
nodes [
 ConversationEntryPointNode_7 {
  id=#1
  class=ConversationEntryPointNode
  outputs [#2]
  properties {entryTag {tagName=Conversation.Entry.DiamondDiva}}
 }
 Speech_C_4 {
  id=#2
  class=Speech_C
  outputs [#3, #4, #5]
  properties {
   nodeName=""
   generalConfig {
    defaultMessage="Let's get straight to business, darling. I've got errands to run."
    contextualMessages [
     {
      message="Hello, darling. The pleasure is all yours."
      contextRequirements [{requirement {class=FortControllerRequirement_IsFirstEverConversationWithNPC}}]
      priority {curve {rowName=Tandem.Speech.Priority.FirstEverInteraction}}
     }
     {
      message=It turns out I'm a fair hand at this heist business. I don't need you... and perhaps I never did.
      contextRequirements [{requirement {class=FortControllerRequirement_HasCID, properties {allowedCIds [/Game/Athena/Items/Cosmetics/Characters/CID_206_Athena_Commando_M_Bling]}}}]
      priority {curve {rowName=Tandem.Speech.Priority.OutfitReaction}}
     }
     {
      message=I heard you're out of the game. Pity... we made a good team.
      contextRequirements [{requirement {
         class=FortControllerRequirement_HasCID
         properties {allowedCIds [/BRCosmetics/Athena/Items/Cosmetics/Characters/CID_397_Athena_Commando_F_TreasureHunterFashion]}
        }}]
      priority {curve {rowName=Tandem.Speech.Priority.OutfitReaction}}
     }
     {
      message="We should team up, darling. I think we could do grand things together."
      contextRequirements [{requirement {
         class=FortControllerRequirement_HasCID
         properties {allowedCIds [/BRCosmetics/Athena/Items/Cosmetics/Characters/CID_A_478_Athena_Commando_F_WildCard]}
        }}]
      priority {curve {rowName=Tandem.Speech.Priority.OutfitReaction}}
     }
     {
      message="Sorry, I'm not looking to join someone else's crew. Bad memories, you understand."
      contextRequirements [{requirement {class=FortControllerRequirement_HasCID, properties {allowedCIds [&Character_MusketSlinger]}}}]
      priority {curve {rowName=Tandem.Speech.Priority.OutfitReaction}}
     }
     {
      message=" I've been told you have a sumptuous collection of antiques. I'd love a personal tour."
      contextRequirements [{requirement {class=FortControllerRequirement_HasCID, properties {allowedCIds [/BRCosmetics/Athena/Items/Cosmetics/Characters/Character_LazarusLens]}}}]
      priority {curve {rowName=Tandem.Speech.Priority.OutfitReaction}}
     }
    ]
   }
  }
 }
 ConversationLinkNode_2 {
  id=#5
  class=ConversationLinkNode
  properties {remoteEntryTag {tagName=Conversation.Entry.Season26.Snapshot.P03.Q07}}
 }
 ConversationLinkNode_0 {id=#4, class=ConversationLinkNode, properties {remoteEntryTag {tagName=Conversation.Entry.BasicServices}}}
 ConversationLinkNode_1 {id=#3, class=ConversationLinkNode, properties {remoteEntryTag {tagName=Conversation.Entry.NPCLibrary.Quests}}}
]
